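# Animation Blueprint for the Lucky Hero backpack cosmetic that adds physics-driven secondary motion to the pack's zipper.
# A bone constraint ties dyn_zipper_base to the pack bone, then Anim Dynamics simulates dyn_zipper_parent with angular limits, high damping and scaled-down component motion.
# A final Modify Bone node applies a small fixed translation and roll to the root.
# The event graph holds no logic, so all behavior lives in the AnimGraph.
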AnimBlueprint /BRCosmetics/Accessories/FORT_Backpacks/Backpack_F_MED_Lucky_Hero/Meshes/F_MED_Lucky_Hero_Pack_AnimBp
target_skeleton=&/F_MED_Lucky_Hero_Pack_Skeleton
graphs [
 AnimGraph {nodes [
   n1 = "Constraint - dyn_zipper_base (P)"!(ComponentPose=Local To Component(LocalPose=Input Pose()), Alpha=1, ConstraintWeights_0=1, ConstraintWeights_1:float=0) -> (Pose) {
    alphaBoolBlend {blendOption=Linear}
    boneToModify {boneName=dyn_zipper_base}
    constraintSetup [{targetBone {boneName=pack}}]
    constraintWeights [1,0]
   }
   n2 = Anim Dynamics - Bone: dyn_zipper_parent(ComponentPose=@n1, Alpha=1, ExternalForce="0,0,0") {
    angularDampingOverride=0.999
    bGravityOverrideInSimSpace=true
    bOverrideAngularDamping=true
    bOverrideLinearDamping=true
    linearDampingOverride=0.99
    alphaBoolBlend {blendOption=Linear}
    boundBone {boneName=dyn_zipper_parent}
    componentAppliedLinearAccClamp=<3000 3000 3000>
    componentLinearAccScale=<0.25 0.25 0.25>
    componentLinearVelScale=<0.25 0.25 0.25>
    physicsBodyDefinitions [{
      sphereCollisionRadius=0
      boundBone {boneName=dyn_zipper_parent}
      boxExtents=<6 4 2>
      constraintSetup {angularLimitsMax=<0 60 30>, angularLimitsMin=<0 -13 -30>, angularTarget=<0 0 0>}
      localJointOffset=<2.5 0 0>
     }]
    relativeSpaceBone {boneName=root}
   }
   n3 = "Transform (Modify) Bone - Bone: root"(ComponentPose=@n2, Translation="0,1,0", Rotation="0,0,-2") {
    rotationMode=BMM_Additive
    translationMode=BMM_Replace
    alphaBoolBlend {blendOption=Linear}
    boneToModify {boneName=root}
    rotation {roll=-2}
    translation {y=1}
   }
   Output Pose(Result=Component To Local(ComponentPose=@n3)) {layerGroup=None}
  ]}
 EventGraph {nodes [
   TryGetPawnOwner()
   Event BlueprintUpdateAnimation()
  ]}
]
variables []
